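# Sound cue named for the S33 military announcement on the train station PA, stored in the S37 folder.
# Its graph has no root or nodes, so as it stands it references no audio.
# Base submix and submix sends are disabled, leaving a pre-effect send to the shared train station announcement audio bus as its configured output.
# It also carries a boosted volume and a very large max distance.
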SoundCue /Train/Sound/Cues/PA_Announcements/S37/SC_Train_Station_PA_S33_Military_Announcement
root=null
nodes []
properties {
 bEnableBaseSubmix=false
 bEnableSubmixSends=false
 maxDistance=2097152
 volumeMultiplier=1.996
 concurrencyOverrides {maxCount=3}
 preEffectBusSends [{audioBus=/Train/Sound/AB_TrainStationAnnouncmentBus, sendLevel=0.4, soundSourceBus=null}]
}
