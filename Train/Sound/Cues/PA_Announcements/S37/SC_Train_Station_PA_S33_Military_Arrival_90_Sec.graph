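# Train station PA sound cue, named for the S33 military arrival 90-second announcement, stored under S37.
# A single Wave Player feeds a Mixer and plays the generic S37 train station jingle stinger, with no separate voice line.
# Base submix and submix sends are disabled, so output routes through a pre-effect send to the station announcement audio bus.
# Volume is boosted, and concurrency override values specify a small instance cap and retrigger time.
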SoundCue /Train/Sound/Cues/PA_Announcements/S37/SC_Train_Station_PA_S33_Military_Arrival_90_Sec
root=Mixer_0
nodes [
 {id=Mixer_0, children [WavePlayer_3], properties {inputVolume [1]}}
 {id=WavePlayer_3, properties {soundWaveAssetPtr=/Train/Sound/Waves/Jingles/S37/MX_S37_POI_TrainStation_Jingle_Generic_STNG}}
]
properties {
 bEnableBaseSubmix=false
 bEnableSubmixSends=false
 maxDistance=2097152
 volumeMultiplier=1.996
 concurrencyOverrides {maxCount=2, retriggerTime=3}
 preEffectBusSends [{audioBus=/Train/Sound/AB_TrainStationAnnouncmentBus, sendLevel=0.4, soundSourceBus=null}]
}
